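# Unlit translucent Niagara sprite material for ember particles in the Headset pickaxe's cosmetic effects.
# A SubUV flipbook texture's red channel, shaped by a power, blends between two color parameters, dark to bright magenta-purple by default.
# The result scales by emissive intensity and particle color, then passes through the shared MF_FX_EmissiveAdjustments function.
# Texture alpha multiplies with particle alpha for transparency.
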Material /BRCosmetics/Weapons/FORT_Melee/Pickaxe_Headset/FX/M_Pickaxe_Headset_Ember
graph {
 properties {bUsedWithNiagaraSprites=true, blendMode=BLEND_Translucent, pixelDepthOffsetMode=PDOM_Legacy, shadingModel=MSM_Unlit, usedShadingModels=Unlit}
 nodes [
  n1 = TextureSampleParameterSubUV!(UVs, Apply View MipBias) {parameterName=Param, texture=&/T_Pickaxe_Headset_Ember_01}
  n2 = LinearInterpolate(
   A=VectorParameter() {parameterName=Color B, defaultValue=<rgba 0 0 0 1>}.RGB
   B=VectorParameter() {parameterName=Color A, defaultValue=<rgba 2.1 0.3 3 1>}.RGB
   Alpha=Power(Base=@n1.R)
  )
  n3 = ScalarParameter() {defaultValue=1, parameterName=Emissive Intensity}
  n4 = ParticleColor()
  MaterialFunctionCall(Emissive=Multiply(A=Multiply(A=@n2, B=@n3), B=@n4.RGB)) {materialFunction=&MF_FX_EmissiveAdjustments}
  Multiply(A=@n1.A, B=@n4.A)
 ]
}
