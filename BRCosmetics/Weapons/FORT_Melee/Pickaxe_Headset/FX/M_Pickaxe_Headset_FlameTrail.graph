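# Unlit translucent, two-sided flame trail material for the Headset pickaxe's Niagara effects.
# Two panning noise textures, warped by normal-map distortion and randomized through dynamic parameters, are shaped by sine tapers and a split mask with optional hard edges.
# That mask drives curve-atlas color, combined with plasma and smoke highlights, particle color and hue shift.
# It also drives opacity, with optional lead fade and depth fade on non-forward, non-mobile paths.
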Material /BRCosmetics/Weapons/FORT_Melee/Pickaxe_Headset/FX/M_Pickaxe_Headset_FlameTrail
graph {
 properties {
  bUsedWithNiagaraMeshParticles=true
  bUsedWithNiagaraRibbons=true
  bUsedWithNiagaraSprites=true
  blendMode=BLEND_Translucent
  pixelDepthOffsetMode=PDOM_Legacy
  shadingModel=MSM_Unlit
  twoSided=true
  usedShadingModels=Unlit
 }
 nodes [
  n1 = TextureCoordinate() {vTiling=2}
  n2 = Power(Base=TextureSample!(UVs=Panner(Coordinate=@n1) {speedX=-0.5}, Tex, Bias=Constant() {r=1}, Apply View MipBias) {mipValueMode=TMVM_MipBias, texture=/Game/Effects/Fort_Effects/Textures/Plasma/T_Plasma_Aurora_Seamless}.RGB)
  n3 = Constant()
  n4 = Multiply(A=Multiply(
    A=Desaturation(None=@n2, Fraction=@n3)
    B=VectorParameter() {group=Emissive, parameterName=Highlight Color A, defaultValue=<rgba 0.1667 0 1 1>}.RGB
   )) {constB=10}
  n5 = DynamicParameter() {defaultValue=<rgba 1 0 1 100>, paramNames [Random Offset, Noise Rotation, WPO Wave Intensity, Sphere Mask Range]}
  n6 = VectorParameter() {group=Noise, parameterName=Noise B UV, sortPriority=6, defaultValue=<rgba 1 1 -3 0.1>}
  n7 = AppendVector(A=@n6.R, B=@n6.R)
  n8 = MaterialFunctionCall(UVs=TextureCoordinate(), "Rotation Angle (0-1)"=@n5.Noise Rotation) {materialFunction=&CustomRotator}
  n9 = Add(
   A=AppendVector(A=@n5.Random Offset, B=@n5.Random Offset)
   B=Panner(Coordinate=Multiply(A=@n8.Rotated Values, B=@n7), Speed=AppendVector(A=@n6.B, B=@n6.B))
  )
  n10 = VectorParameter() {group=Distortion, parameterName=Distortion UV, sortPriority=1, defaultValue=<rgba 1 1 -1 0>}
  n11 = Multiply(A=@n8.Rotated Values, B=AppendVector(A=@n10.R, B=@n10.R))
  n12 = AppendVector(A=@n10.B, B=@n10.B)
  n13 = TextureSampleParameter2D(UVs=Add(A=AppendVector(A=@n5.Random Offset, B=@n5.Random Offset), B=Panner(Coordinate=@n11, Speed=@n12))) {
   group=Distortion
   parameterName="Distortion  Tex"
   samplerType=SAMPLERTYPE_Normal
   sortPriority=0
   texture=&T_Weap_Distort_N_02
  }
  n14 = OneMinus(None=ComponentMask(None=@n8.Rotated Values) {g=true})
  n15 = ScalarParameter() {defaultValue=0.5, group=Distortion, parameterName=Distortion Gradient Lerp, sortPriority=3}
  n16 = Multiply(
   A=Multiply(A=ComponentMask(None=@n13.RGB) {g=true, r=true}, B=LinearInterpolate(B=@n14, Alpha=@n15) {constA=1, constAlpha=0})
   B=ScalarParameter() {defaultValue=0.3, group=Distortion, parameterName=Distortion Intensity, sortPriority=2}
  )
  n17 = ChannelMaskParameter(None=TextureSampleParameter2D(UVs=Add(A=@n9, B=@n16)) {group=Noise, parameterName=NoIse B Texture, sortPriority=4, texture=&T_NoisePack_01}.RGB) {group=Noise, maskChannel=Green, parameterName=Noise B Channel, sortPriority=5, defaultValue=<rgba 0 1 0 0>}
  n18 = Saturate(None=Multiply(A=@n17.R, B=ScalarParameter() {defaultValue=1, group=Noise, parameterName=NoIse B Intensity, sortPriority=7}) {constB=10})
  n19 = AppendVector(A=@n5.Random Offset, B=@n5.Random Offset)
  n20 = VectorParameter() {group=Noise, parameterName=Noise A UV, sortPriority=2, defaultValue=<rgba 1 1.5 -2 -0.2>}
  n21 = Panner(Coordinate=Multiply(A=@n8.Rotated Values, B=AppendVector(A=@n20.R, B=@n20.R)), Speed=AppendVector(A=@n20.B, B=@n20.B))
  n22 = TextureSampleParameter2D(UVs=Add(A=@n16, B=Add(A=@n19, B=@n21))) {
   group=Noise
   parameterName=NoIse A Texture
   sortPriority=0
   texture=&/T_Pickaxe_Headset_Noise_01
  }
  n23 = Saturate(None=Multiply(
    A=ChannelMaskParameter(None=@n22.RGB) {group=Noise, parameterName=Noise A Channel, sortPriority=1}.R
    B=ScalarParameter() {defaultValue=5, group=Noise, parameterName=NoIse A Intensity, sortPriority=3}
   ) {constB=10})
  n24 = ScalarParameter() {defaultValue=0.5, group=Noise, parameterName=Lerp Noise AB, sortPriority=8}
  n25 = MaterialFunctionCall(Float2=TextureCoordinate()) {materialFunction=&BreakOutFloat2Components}
  n26 = Multiply(A=@n25.G) {constB=0.5}
  n27 = ScalarParameter() {defaultValue=4, group=Shape, parameterName=Top Thickness, sortPriority=0}
  n28 = ScalarParameter() {defaultValue=1, group=Shape, parameterName=Bottom Thickness, sortPriority=1}
  n29 = OneMinus(None=@n25.R)
  n30 = Multiply(A=Power(Base=Sine(None=@n26), Exp=LinearInterpolate(A=@n27, B=@n28, Alpha=@n29) {constA=6, constB=2}) {constExponent=4}, B=@n29)
  n31 = Power(Base=@n30)
  n32 = Multiply(A=Add(A=LinearInterpolate(A=@n18, B=@n23, Alpha=@n24), B=Saturate(None=@n31)), B=@n30)
  n33 = LinearInterpolate(
   A=ScalarParameter() {defaultValue=1, group=Shape, parameterName=Top Fade, sortPriority=2}
   B=ScalarParameter() {defaultValue=1, group=Shape, parameterName=Bottom Fade, sortPriority=3}
   Alpha=@n29
  ) {constA=1, constB=0.5}
  n34 = TextureCoordinate()
  n35 = CurveAtlasRowParameter(CurveTime=OneMinus(None=ComponentMask(None=@n34) {r=true})) {
   atlas=/Game/Athena/LinearColorCurves/A_Atlas_Turtleneck_Glider
   curve=/Game/Weapons/FORT_Melee/Pickaxe_Headset/FX/Curve_Pickaxe_Headset_Flame
   defaultValue=31
   group=Emissive
   parameterName=ColorCurve
   sortPriority=0
  }
  n36 = Sine(None=Multiply(A=@n25.G, B=ScalarParameter() {defaultValue=1, group=Shape, parameterName=Fire Split Offset, sortPriority=9}) {constB=1.5}) {period=2}
  n37 = Add(
   A=Power(Base=OneMinus(None=@n36), Exp=LinearInterpolate(Alpha=@n13.R) {constA=0.5, constB=1.5}) {constExponent=0.5}
   B=Power(Base=@n29, Exp=ScalarParameter() {defaultValue=0.8, group=Shape, parameterName=Fire Split Contrast, sortPriority=8})
  )
  n38 = Multiply(
   A=Multiply(A=Power(Base=@n32, Exp=@n33), B=@n35.A)
   B=LinearInterpolate(B=Round(None=@n37), Alpha=ScalarParameter() {defaultValue=0.5, group=Shape, parameterName=Fire Split, sortPriority=7}) {constA=1, constB=0}
  )
  n39 = Constant() {r=1}
  n40 = If(
   A=@n38
   B=OneMinus(None=ScalarParameter() {group=Shape, parameterName=Hard Shape, sortPriority=4})
   "A > B"=@n39
   "A == B"=@n39
   "A < B"=Constant()
  )
  n41 = LinearInterpolate(A=@n38, B=@n40, Alpha=ScalarParameter() {defaultValue=0.02, group=Shape, parameterName=Alpha Lerp Hard Shape, sortPriority=6})
  n42 = CurveAtlasRowParameter(CurveTime=@n41) {
   atlas=&A_Atlas_Turtleneck_Glider
   curve=&Curve_Pickaxe_Headset_Flame
   defaultValue=31
   group=Emissive
   parameterName=ColorCurve
   sortPriority=0
  }
  n43 = ScalarParameter() {defaultValue=1, group=Emissive, parameterName=ColorCurve_Lerp}
  n44 = LinearInterpolate(A=@n38, B=@n40, Alpha=ScalarParameter() {group=Shape, parameterName=Emissive Lerp Hard Shape, sortPriority=5})
  n45 = Constant3Vector() {constant=<rgba 2 1 2 0>}
  n46 = Multiply(A=Add(A=LinearInterpolate(A=@n41, B=@n42, Alpha=@n43), B=Multiply(A=@n44, B=@n45)))
  n47 = VectorParameter() {group=Emissive, parameterName=Highlight Color B, defaultValue=<rgba 50.0001 0 300 1>}
  n48 = TextureCoordinate()
  n49 = AppendVector(A=ComponentMask(None=@n48) {r=true}, B=Add(A=ComponentMask(None=@n48) {g=true}, B=ParticleRandom()))
  n50 = Power(Base=TextureSample!(UVs=Panner(Coordinate=@n49) {speedX=-1}, Tex, Apply View MipBias) {texture=&T_Inky_Smoke_Tile}.R) {constExponent=10}
  n51 = Multiply(
   A=Add(A=Add(A=@n4, B=@n46), B=Multiply(A=@n47.RGB, B=@n50))
   B=ScalarParameter() {defaultValue=1, group=Emissive, parameterName=Emissive_Intensity}
  )
  n52 = ParticleColor()
  MaterialFunctionCall(Emissive=MaterialFunctionCall(Hue Shift Percentage=ScalarParameter() {parameterName=HSP}, Texture=Multiply(A=@n51, B=@n52.RGB)) {materialFunction=&HueShift}.Result) {materialFunction=&MF_FX_EmissiveAdjustments}
  n53 = VectorParameter() {
   group=BottomFade
   parameterName=Bottom Fade_SphereMask_Scale
   sortPriority=1
   channelNames {a=Center V, b=Center U, g=Scale V, r=Scale U}
   defaultValue=<rgba 1 1 0 0.5>
  }
  n54 = SphereMask(
   A=Multiply(A=TextureCoordinate(), B=AppendVector(A=@n53.Scale U, B=@n53.Scale U))
   B=AppendVector(A=@n53.Center U, B=@n53.Center U)
   Radius=ScalarParameter() {defaultValue=1, group=BottomFade, parameterName=BottomFade_SphereMask_Radius, sortPriority=2}
  ) {attenuationRadius=0.95, hardnessPercent=0}
  n55 = ScalarParameter() {defaultValue=1, group=BottomFade, parameterName=BottomFade_SphereMask_Power, sortPriority=3}
  Multiply(A=@n35.A, B=Saturate(None=Power(Base=@n54, Exp=@n55)))
  n56 = ScalarParameter() {defaultValue=10, group=Opacity, parameterName=Opacity Multiplier, sortPriority=1}
  n57 = TextureCoordinate()
  n58 = Multiply(
   A=Power(Base=ComponentMask(None=@n57) {r=true}, Exp=ScalarParameter() {defaultValue=2, group=Opacity, parameterName=LeadFadePower})
   B=ScalarParameter() {defaultValue=3, group=Opacity, parameterName=LeadFade EdgeIntensity}
  ) {constB=3}
  n59 = Multiply(
   A=Multiply(A=@n52.A, B=Multiply(A=@n41, B=@n56) {constB=3})
   B=LinearInterpolate(B=Saturate(None=@n58), Alpha=ScalarParameter() {group=Opacity, parameterName=Use Lead Fade}) {constA=1}
  )
  n60 = DepthFade(Opacity=@n59, FadeDistance=ScalarParameter() {parameterName=DepthFade})
  ShadingPathSwitch(Default=@n60, Deferred=@n60, Forward=@n59, Mobile=@n59)
 ]
}
