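# Physics asset for the Headset pickaxe, a two-handed melee cosmetic, simulating two dangling straps on the weapon mesh.
# The handle body uses two capsules and a box, while each strap is a chain of three capsule bodies with swing and twist limits.
# Strap A attaches directly to the handle.
# Strap B hangs from a small sphere master body joined to the handle with free swing and twist.
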PhysicsAsset /BRCosmetics/Weapons/FORT_Melee/Pickaxe_Headset/Meshes/Headset_Axe_Physics
bodies {
 Melee_TwoHanded_Handle [
  {
   shapeType=Capsule
   length=5.8442
   radius=2.1622
   shapeName=Melee_TwoHanded_Handle_capsule_C
   center=<-0.0333 0.1404 -10.8859>
   rotation=<rot -0.0173 0 -62.6372>
  }
  {shapeType=Capsule, length=1.8709, radius=4.6073, shapeName=Melee_TwoHanded_Handle_capsule, center=<0 1.5047 -0.4848>, rotation=<rot 0 0 90>}
  {shapeType=Box, extentX=2.565, extentY=10.4006, extentZ=12.2016, shapeName=Melee_TwoHanded_Handle_box, center=<0.243 3.1594 -5.3183>}
 ]
 dyn_strap_A_1 [{
   shapeType=Capsule
   length=1.6866
   radius=0.5
   shapeName=dyn_strap_A_1_capsule
   center=<1.6361 0.26 -0.117>
   rotation=<rot 85.0722 176.2883 179.9999>
  }]
 dyn_strap_A_2 [{shapeType=Capsule, length=2.8463, radius=0.6445, shapeName=dyn_strap_A_2_capsule, center=<1.6902 0 0>, rotation=<rot 90 0 0>}]
 dyn_strap_A_3 [{shapeType=Capsule, length=3.1977, radius=0.7237, shapeName=dyn_strap_A_2_capsule, center=<1.9542 0 0>, rotation=<rot 90 0 0>}]
 dyn_strap_B_1 [{shapeType=Capsule, length=1.4939, radius=0.5, shapeName=dyn_strap_B_1_capsule, center=<1.5179 0.15 0>, rotation=<rot 90 -43.2439 -37.3416>}]
 dyn_strap_B_2 [{shapeType=Capsule, length=3.6545, radius=0.6563, shapeName=dyn_strap_A_2_capsule, center=<2.4736 0 0>, rotation=<rot 90 0 0>}]
 dyn_strap_B_3 [{shapeType=Capsule, length=3.9201, radius=0.6994, shapeName=dyn_strap_A_2_capsule, center=<2.5197 0 0>, rotation=<rot 90 0 0>}]
 dyn_strap_B_MSTR [{shapeType=Sphere, radius=0.505, shapeName=dyn_strap_B_MSTR_sphere}]
}
constraints [
 {bone1Name=dyn_strap_B_1, bone2Name=dyn_strap_B_MSTR, swing1LimitDegrees=40, swing2LimitDegrees=40, twistLimitDegrees=10}
 {bone1Name=dyn_strap_B_2, bone2Name=dyn_strap_B_1, swing1LimitDegrees=45, swing2LimitDegrees=45, twistLimitDegrees=10}
 {bone1Name=dyn_strap_B_3, bone2Name=dyn_strap_B_2, swing1LimitDegrees=45, swing2LimitDegrees=45, twistLimitDegrees=10}
 {bone1Name=dyn_strap_A_1, bone2Name=Melee_TwoHanded_Handle, swing1LimitDegrees=50, swing2LimitDegrees=60, twistLimitDegrees=10}
 {bone1Name=dyn_strap_A_2, bone2Name=dyn_strap_A_1, swing1LimitDegrees=45, swing2LimitDegrees=45, twistLimitDegrees=10}
 {bone1Name=dyn_strap_A_3, bone2Name=dyn_strap_A_2, swing1LimitDegrees=45, swing2LimitDegrees=45, twistLimitDegrees=10}
 {
  bone1Name=dyn_strap_B_MSTR
  bone2Name=Melee_TwoHanded_Handle
  swing1LimitDegrees=45
  swing1Motion=Free
  swing2LimitDegrees=45
  swing2Motion=Free
  twistLimitDegrees=45
  twistMotion=Free
 }
]
