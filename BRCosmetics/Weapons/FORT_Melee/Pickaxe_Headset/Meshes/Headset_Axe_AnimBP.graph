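# Animation Blueprint for the Headset pickaxe that adds physics motion to its dynamic strap on top of the copied parent pose.
# It caches the pose copied from the attached parent mesh, nudges the strap master bone additively, then runs RigidBody simulation with overridden gravity and damped component motion.
# A timed bool blend on bClothEnabled switches between the cached pose and the simulated result.
# The event graph holds no active logic.
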AnimBlueprint /BRCosmetics/Weapons/FORT_Melee/Pickaxe_Headset/Meshes/Headset_Axe_AnimBP
target_skeleton=&/Headset_Axe_Skeleton
graphs [
 AnimGraph {nodes [
   n1 = "Transform (Modify) Bone - Bone: dyn_strap_B_MSTR"(
    ComponentPose=Local To Component(LocalPose=Use cached pose 'Start_Cache'())
    Alpha=1
    Translation="-1,-0.000,0.000"
    Rotation="0,0,0"
    Scale="1,1,1"
   ) {
    rotationMode=BMM_Additive
    rotationSpace=BCS_BoneSpace
    translationMode=BMM_Additive
    translationSpace=BCS_BoneSpace
    alphaBoolBlend {blendOption=Linear}
    boneToModify {boneName=dyn_strap_B_MSTR}
    translation {x=-1}
   }
   n2 = RigidBody(ComponentPose=@n1) {
    bOverrideWorldGravity=true
    overridePhysicsAsset=&/Headset_Axe_Physics
    alphaBoolBlend {blendOption=Linear}
    componentLinearAccScale=<0.3 0.3 0.3>
    componentLinearVelScale=<0.3 0.3 0.3>
    overrideWorldGravity {z=-980}
   }
   Output Pose(Result=Blend Poses by bool(
     BlendPose_0=Use cached pose 'Start_Cache'()
     BlendPose_1=Component To Local(ComponentPose=@n2)
     BlendTime_0=0.2
     BlendTime_1=0.5
     bActiveValue=Get bClothEnabled()
    ) {blendType=Linear}) {layerGroup=None}
   Save cached pose 'Start_Cache'(Pose=Copy Pose From Mesh() {bUseAttachedParent=true}) {cacheName=Start_Cache}
  ]}
 EventGraph {nodes [
   TryGetPawnOwner()
   Event BlueprintUpdateAnimation()
  ]}
]
variables []
