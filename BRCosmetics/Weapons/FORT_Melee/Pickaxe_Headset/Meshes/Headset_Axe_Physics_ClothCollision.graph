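# Physics asset for the Headset pickaxe melee weapon, named as cloth collision for its mesh.
# It holds a single body on the Melee_TwoHanded_Handle bone, built from three overlapping capsules at different offsets and angles.
# There are no constraints, so it defines only static collision shapes relative to that bone rather than a simulated ragdoll.
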PhysicsAsset /BRCosmetics/Weapons/FORT_Melee/Pickaxe_Headset/Meshes/Headset_Axe_Physics_ClothCollision
bodies {Melee_TwoHanded_Handle [
  {
   shapeType=Capsule
   length=4.3087
   radius=5.9641
   shapeName=Melee_TwoHanded_Handle_capsule
   center=<4.5737 0.651 -9.003>
   rotation=<rot -14.7746 0 0>
  }
  {
   shapeType=Capsule
   length=6.3931
   radius=7.7231
   shapeName=Melee_TwoHanded_Handle_capsule
   center=<0.0001 4.819 -0.4294>
   rotation=<rot -0.0086 0 109.4324>
  }
  {shapeType=Capsule, length=7.17, radius=4.8593, shapeName=Melee_TwoHanded_Handle_capsule, center=<0 6.1013 -9.1469>}
 ]}
constraints []
